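# Debug unlit material in the SubUV Maker tools that generates a test velocity pattern: a hard-edged ring of radial 2D vectors around the UV center.
# Normalized centered UVs, with Y flipped, give the direction.
# Distance from center, thresholded by two radius parameters through Ceil and Saturate, forms the ring mask.
# The masked vector is scaled by constants, and negative emissive is allowed so signed values can be kept.
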Material /Volumetrics/Tools/SubUVMaker/Materials/Debug/M_VelocityCone
graph {
 properties {bAllowNegativeEmissiveColor=true, pixelDepthOffsetMode=PDOM_Legacy, shadingModel=MSM_Unlit, usedShadingModels=Unlit}
 nodes [
  n1 = Subtract(A=TextureCoordinate()) {constB=0.5}
  n2 = Multiply(A=Constant2Vector() {g=-1, r=1}, B=Normalize(VectorInput=@n1))
  n3 = MaterialFunctionCall(Vector 2=@n1) {materialFunction=/Engine/Functions/Engine_MaterialFunctions02/Utility/VectorLength}
  n4 = Subtract(A=@n3.V2 Length, B=ScalarParameter() {defaultValue=0.1, parameterName=asdfa2})
  n5 = Ceil(None=Subtract(A=@n3.V2 Length, B=ScalarParameter() {defaultValue=0.2, parameterName=asdfa}))
  n6 = Multiply(A=Saturate(None=Ceil(None=@n4)), B=OneMinus(None=Saturate(None=@n5)))
  Multiply(A=Divide(A=Multiply(A=@n2, B=@n6)) {constB=512}) {constB=10}
 ]
}
